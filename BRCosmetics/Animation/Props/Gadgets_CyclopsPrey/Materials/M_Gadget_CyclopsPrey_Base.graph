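# Base material for the CyclopsPrey gadget emote prop, adding animated burger cooking and chalk sign lines to the shared 2023 anime character shader.
# Channel masks isolate three burgers, whose gradient-driven reveals follow a cooking scalar with flip blend or timestamp-based timers.
# Burger alphas appear to blend raw and shader-input lit/shaded colors, with chalk tints layered through UV1 sign textures chosen by a variation parameter.
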Material /BRCosmetics/Animation/Props/Gadgets_CyclopsPrey/Materials/M_Gadget_CyclopsPrey_Base
graph {
 properties {
  bAutomaticallySetUsageInEditor=false
  bUseMaterialAttributes=true
  bUsedWithMeshDeformer=true
  bUsedWithMorphTargets=true
  bUsedWithSkeletalMesh=true
  bUsedWithStaticLighting=true
  pixelDepthOffsetMode=PDOM_Legacy
  shadingModel=MSM_Cloth
  usedShadingModels=Cloth
 }
 nodes [
  n1 = MaterialFunctionCall!(CustomUV) {materialFunction=/Game/Global/Materials/Character_Parent_2023/Anime_Parent_2023/MaterialFunctions/MFC_AnimeShader_2023_Pre}
  n2 = MaterialFunctionCall!(CustomLightVector, OverrideNormal_WS, STT=@n1.STT, Normal_Map=@n1.STT, DFL=@n1.STT, FX=@n1.STT) {materialFunction=/Game/Global/Materials/Character_Parent_2023/Anime_Parent_2023/MaterialFunctions/MFC_AnimeShader_2023_Calculations}
  MaterialFunctionCall!(
   Shading_Gradient=@n2.Shading_Gradient
   Threshold_Gradient=@n2.Shading_Gradient
   Specular_Gradient=@n2.Shading_Gradient
   Sheen_Mask=@n2.Shading_Gradient
   Thickness=@n2.Shading_Gradient
   DFL_Mask=@n2.Shading_Gradient
   FX=@n2.Shading_Gradient
   Metallic_Gradient=@n2.Shading_Gradient
   RimLight_Gradient=@n2.Shading_Gradient
   NormalBuffer=@n2.Shading_Gradient
   --------
   CINE_LightSources=@n2.Shading_Gradient
   CINE_LightColor=@n2.Shading_Gradient
   ---------
   Color_Lit=NamedRerouteUsage()
   Color_Shaded=NamedRerouteUsage()
   SRM=@n1.SRM
   Emissive=@n1.SRM
   DFL_Color=@n1.SRM
   Specular_Color=@n1.SRM
   RimLight_Color=@n1.SRM
  ) {materialFunction=/Game/Global/Materials/Character_Parent_2023/Anime_Parent_2023/MaterialFunctions/MFC_AnimeShader_2023_Post}
  n3 = ScalarParameter() {group=CyclopsPrey, parameterName=Burger_A_Cooking_Animation}
  Multiply(A=@n3, B=NamedRerouteUsage())
  n4 = TextureSampleParameter2D() {
   group=CyclopsPrey
   parameterName=BurgerGradient
   samplerType=SAMPLERTYPE_LinearColor
   texture=/BRCosmetics/Animation/Props/Gadgets_CyclopsPrey/Textures/T_CyclopsPrey_Emote_Prop_Gradient
  }
  NamedRerouteDeclaration(None=@n4.RGB) {name=BurgerGradient, nodeColor=<rgba 1 0.7031 0 1>}
  n5 = TextureCoordinate() {coordinateIndex=1}
  n6 = TextureSampleParameter2D(UVs=@n5) {
   group=CyclopsPrey
   parameterName=SignTexture_A
   samplerType=SAMPLERTYPE_Alpha
   texture=/BRCosmetics/Animation/Props/Gadgets_CyclopsPrey/Textures/T_CyclopsPrey_Emote_Prop_Sign_A
  }
  NamedRerouteDeclaration(None=@n6.RGB) {name=SignTexture_A, nodeColor=<rgba 0 1 0.4375 1>}
  n7 = ScalarParameter() {group=CyclopsPrey, parameterName=Burger_C_TimeStamp}
  n8 = Divide(A=Subtract(A=Time(), B=@n7), B=ScalarParameter() {defaultValue=10, group=CyclopsPrey, parameterName=Burger_C_AnimLength})
  n9 = Multiply(A=Multiply(A=Saturate(None=@n8), B=Saturate(None=@n7))) {constB=3}
  n10 = Saturate(None=Subtract(A=@n9) {constB=2})
  n11 = MaterialFunctionCall(Float3=NamedRerouteUsage()) {materialFunction=&BreakOutFloat3Components}
  n12 = Subtract(A=@n9)
  n13 = OneMinus(None=@n11.G)
  n14 = Max(A=Add(A=Saturate(None=@n9), B=Subtract(A=@n11.R)), B=Add(A=Saturate(None=@n12), B=Subtract(A=@n13)))
  n15 = Add(A=@n10, B=Subtract(A=@n11.B))
  n16 = Multiply(A=NamedRerouteUsage(), B=Add(A=@n10, B=Max(A=@n14, B=@n15)))
  NamedRerouteDeclaration(None=Saturate(None=Power(Base=@n16) {constExponent=0.5})) {name=LeftBurger_Alpha, nodeColor=<rgba 1 0.5391 0 1>}
  n17 = ScalarParameter() {group=CyclopsPrey, parameterName=Burger_B_TimeStamp}
  n18 = Divide(A=Subtract(A=Time(), B=@n17), B=ScalarParameter() {defaultValue=10, group=CyclopsPrey, parameterName=Burger_B_AnimLength})
  n19 = Multiply(A=Multiply(A=Saturate(None=@n18), B=Saturate(None=@n17))) {constB=3}
  n20 = Saturate(None=Subtract(A=@n19) {constB=2})
  n21 = MaterialFunctionCall(Float3=NamedRerouteUsage()) {materialFunction=&BreakOutFloat3Components}
  n22 = OneMinus(None=@n21.R)
  n23 = Subtract(A=@n19)
  n24 = Max(A=Add(A=Saturate(None=@n19), B=Subtract(A=@n22)), B=Add(A=Saturate(None=@n23), B=Subtract(A=@n21.G)))
  n25 = Add(A=@n20, B=Subtract(A=@n21.B))
  n26 = Multiply(A=NamedRerouteUsage(), B=Add(A=@n20, B=Max(A=@n24, B=@n25)))
  NamedRerouteDeclaration(None=Saturate(None=Power(Base=@n26) {constExponent=0.5})) {name=RightBurger_Alpha, nodeColor=<rgba 0 1 0.0156 1>}
  n27 = Multiply(A=@n3) {constB=3}
  n28 = Saturate(None=Subtract(A=@n27) {constB=2})
  n29 = Saturate(None=Subtract(A=@n27))
  n30 = MaterialFunctionCall(Float3=NamedRerouteUsage()) {materialFunction=&BreakOutFloat3Components}
  n31 = Subtract(A=OneMinus(None=@n30.R))
  n32 = Saturate(None=@n27)
  n33 = Subtract(A=@n30.G)
  n34 = LinearInterpolate(
   A=Max(A=Add(A=@n29, B=@n31), B=Add(A=@n32, B=@n33))
   B=Max(A=Add(A=@n32, B=@n31), B=Add(A=@n29, B=@n33))
   Alpha=ScalarParameter() {group=CyclopsPrey, parameterName=Burger_A_Flip_Animation}
  )
  n35 = Add(A=@n28, B=Subtract(A=@n30.B))
  n36 = Multiply(A=NamedRerouteUsage(), B=Add(A=@n28, B=Max(A=@n34, B=@n35)))
  NamedRerouteDeclaration(None=Saturate(None=Power(Base=@n36) {constExponent=0.5})) {name=MiddleBurger_Alpha, nodeColor=<rgba 0.4297 0 1 1>}
  n37 = TextureSampleParameter2D(UVs=@n5) {
   group=CyclopsPrey
   parameterName=SignTexture_B
   samplerType=SAMPLERTYPE_Alpha
   texture=/BRCosmetics/Animation/Props/Gadgets_CyclopsPrey/Textures/T_CyclopsPrey_Emote_Prop_Sign_B
  }
  NamedRerouteDeclaration(None=@n37.RGB) {name=SignTexture_B, nodeColor=<rgba 1 0.1172 0 1>}
  n38 = TextureSampleParameter2D(UVs=@n5) {
   group=CyclopsPrey
   parameterName=SignTexture_C
   samplerType=SAMPLERTYPE_Alpha
   texture=/BRCosmetics/Animation/Props/Gadgets_CyclopsPrey/Textures/T_CyclopsPrey_Emote_Prop_Sign_C
  }
  NamedRerouteDeclaration(None=@n38.RGB) {name=SignTexture_C, nodeColor=<rgba 0 0.2031 1 1>}
  n39 = TextureSampleParameter2D() {
   group=CyclopsPrey
   parameterName=BurguerMasks
   samplerType=SAMPLERTYPE_Masks
   texture=/BRCosmetics/Animation/Props/Gadgets_CyclopsPrey/Textures/T_CyclopsPrey_Emote_Prop_FX
  }
  n40 = NamedRerouteDeclaration(None=@n39.R) {name=MiddleBurger_Mask, nodeColor=<rgba 1 0.3047 0 1>}
  n41 = NamedRerouteDeclaration(None=@n39.G) {name=LeftBurger_Mask, nodeColor=<rgba 0.3359 1 0 1>}
  NamedRerouteDeclaration(None=Max(A=Max(A=@n40, B=@n41), B=NamedRerouteDeclaration(None=@n39.B) {name=RightBurger_Mask, nodeColor=<rgba 0 0.0156 1 1>})) {name=Burgers_Mask, nodeColor=<rgba 0 1 0.6016 1>}
  NamedRerouteDeclaration(None=@n1.Color_Lit) {name=ColorLit_IN, nodeColor=<rgba 0.6172 0 1 1>}
  NamedRerouteDeclaration(None=@n1.Color_Shaded) {name=ColorShaded_IN, nodeColor=<rgba 0 1 0.2969 1>}
  n42 = NamedRerouteUsage()
  n43 = NamedRerouteUsage()
  n44 = Max(A=NamedRerouteUsage(), B=NamedRerouteUsage())
  n45 = NamedRerouteUsage()
  n46 = NamedRerouteUsage()
  n47 = Saturate(None=Add(A=Max(A=@n44, B=@n45), B=OneMinus(None=@n46)))
  NamedRerouteDeclaration(None=LinearInterpolate(
    A=LinearInterpolate(A=@n42, B=@n43, Alpha=@n47) {constAlpha=0}
    B=VectorParameter() {group=CyclopsPrey, parameterName=ChalkLitColor, defaultValue=<rgba 0.8 0.8 0.8 1>}.RGB
    Alpha=NamedRerouteUsage()
   )) {name=ColorLit_OUT, nodeColor=<rgba 0.1953 1 0 1>}
  n48 = NamedRerouteUsage()
  n49 = NamedRerouteUsage()
  n50 = Max(A=NamedRerouteUsage(), B=NamedRerouteUsage())
  n51 = NamedRerouteUsage()
  n52 = NamedRerouteUsage()
  n53 = Saturate(None=Add(A=Max(A=@n50, B=@n51), B=OneMinus(None=@n52)))
  NamedRerouteDeclaration(None=LinearInterpolate(
    A=LinearInterpolate(A=@n48, B=@n49, Alpha=@n53) {constAlpha=0}
    B=VectorParameter() {group=CyclopsPrey, parameterName=ChalkShadedColor, defaultValue=<rgba 0.5 0.6 0.6 1>}.RGB
    Alpha=NamedRerouteUsage()
   )) {name=ColorShaded_OUT, nodeColor=<rgba 1 0 0.1172 1>}
  NamedRerouteDeclaration(None=If(
    A=ScalarParameter() {group=CyclopsPrey, parameterName=Sign_Variation}
    "A > B"=NamedRerouteUsage()
    "A == B"=NamedRerouteUsage()
    "A < B"=NamedRerouteUsage()
   ) {constB=0.5}) {name=SignLines_Alpha, nodeColor=<rgba 0.2656 0 1 1>}
  n54 = TextureSampleParameter2D() {
   group=CyclopsPrey
   parameterName=RawColor_Lit
   texture=/BRCosmetics/Animation/Props/Gadgets_CyclopsPrey/Textures/T_CyclopsPrey_BurgerPink_Emote_Prop_CL
  }
  NamedRerouteDeclaration(None=@n54.RGB) {name=RawColor_Lit, nodeColor=<rgba 1 0 0.6562 1>}
  n55 = TextureSampleParameter2D() {
   group=CyclopsPrey
   parameterName=RawColor_Shaded
   texture=/BRCosmetics/Animation/Props/Gadgets_CyclopsPrey/Textures/T_CyclopsPrey_BurgerPink_Emote_Prop_CS
  }
  NamedRerouteDeclaration(None=@n55.RGB) {name=RawColor_Shaded, nodeColor=<rgba 0 0.9766 1 1>}
 ]
}
